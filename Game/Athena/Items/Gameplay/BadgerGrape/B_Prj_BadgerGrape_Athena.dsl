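; BadgerGrape hook-gun projectile for Athena, a child Blueprint that supplies its own rope and quest-credit behavior.
; SpawnRope spawns the BadgerGrape rope actor at the projectile's location with the player pawn as instigator, stores it, and links it to the weapon and projectile.
; GiveQuestCredit runs the parent logic, then applies the PostBadgerTag quest GameplayEffect to the player pawn's ability system for quest-progress checks.
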(fn GiveQuestCredit
  ;[1 1 1] Apply Timed GE and check for Quest Progress
  (|CallParentFunction)
  (GameplayEffects|ApplyGameplayEffectToSelf :self (Ability|GetAbilitySystemComponent (Variables|Default|GetPlayerPawn)) :GameplayEffectClass "/Game/Athena/Items/QuestInteractables/Quest_GEs/GE_Athena_Quests_PostBadgerTag.GE_Athena_Quests_PostBadgerTag_C" :Level 0.0))

(fn SpawnRope
  (params (PlayerPawn :type Fort%20Player%20Pawn%20Object%20Reference))
  ;[1 1 1] Spawn The Rope
  (bind _returnvalue (Game|SpawnActorfromClass "/Game/Athena/Items/Gameplay/BadgerGrape/B_BadgerGrapeRope_Athena.B_BadgerGrapeRope_Athena_C" (Math|Transform|MakeTransform (Transformation|GetActorLocation)) "Undefined" "MultiplyWithRoot" (Variables|Default|GetWeapon%20Actor) PlayerPawn 0))
  (bind _output_get (Variables|Default|SetRope _returnvalue))
  (Class|FortHookGunRope|SetWeaponandProjectile _output_get (Variables|Default|GetWeapon%20Actor) self)
  (return _output_get))

(fn ConstructionScript
  (|CallParentFunction))
